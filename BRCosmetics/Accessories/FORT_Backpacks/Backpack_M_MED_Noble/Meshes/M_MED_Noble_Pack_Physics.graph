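# Physics asset for the M_MED_Noble backpack cosmetic, giving its dangling accessories simulated secondary motion.
# A box body covers the pack, with capsules on the evaporator, control stick, knife, lightsaber, strap and two belt segments.
# The accessories are jointed to the pack with tight swing and twist limits.
# Two short chains add looser links: the strap hangs from the lightsaber, and the second belt segment follows the first.
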PhysicsAsset /BRCosmetics/Accessories/FORT_Backpacks/Backpack_M_MED_Noble/Meshes/M_MED_Noble_Pack_Physics
bodies {
 dyn_belt_01 [{shapeType=Capsule, length=1.2423, radius=2.2919, shapeName=dyn_belt_01_capsule, center=<2.6503 0 0>, rotation=<rot 90 0 0>}]
 dyn_belt_2 [{shapeType=Capsule, length=2.1677, radius=2.0785, shapeName=dyn_belt_2_capsule, center=<2.0794 0 0>, rotation=<rot 90 0 0>}]
 dyn_control_stick [{
   shapeType=Capsule
   length=11.9277
   radius=4.6336
   shapeName=dyn_control_stick_capsule
   center=<0.6547 -2.0428 -1.3286>
   rotation=<rot 63.7683 -179.9998 180>
  }]
 dyn_evaporator [{
   shapeType=Capsule
   length=35.2106
   radius=6.0766
   shapeName=dyn_evaporator_capsule
   center=<-7.1968 0 0.3376>
   rotation=<rot 90 179.9604 179.9599>
  }]
 dyn_knife [{
   shapeType=Capsule
   length=26.8857
   radius=2.4415
   shapeName=dyn_knife_capsule
   center=<-0.9068 -0.2575 -1.9306>
   rotation=<rot 0 -91.982 -89.9251>
  }]
 dyn_lightsaber1 [{
   shapeType=Capsule
   length=27.8335
   radius=2.7832
   shapeName=dyn_lightsaber1_capsule
   center=<7.9416 0.9891 0.1524>
   rotation=<rot 89.7172 2.8686 -0.0046>
  }]
 dyn_strap_01 [{shapeType=Capsule, length=0.1257, radius=1.6349, shapeName=dyn_strap_01_capsule, center=<0 0 -1.4958>, rotation=<rot 90 0 0>}]
 pack [{shapeType=Box, extentX=21.2271, extentY=12.3192, extentZ=32.5758, shapeName=pack_box, center=<-0.5608 -8.0513 -7.2857>}]
}
constraints [
 {bone1Name=dyn_evaporator, bone2Name=pack, swing1LimitDegrees=5, swing2LimitDegrees=10, twistLimitDegrees=5}
 {bone1Name=dyn_control_stick, bone2Name=pack, swing1LimitDegrees=25, swing2LimitDegrees=12, twistLimitDegrees=2}
 {bone1Name=dyn_knife, bone2Name=pack, swing1LimitDegrees=15, swing2LimitDegrees=15, twistLimitDegrees=2}
 {bone1Name=dyn_lightsaber1, bone2Name=pack, swing1LimitDegrees=10, swing2LimitDegrees=10, twistLimitDegrees=2}
 {bone1Name=dyn_strap_01, bone2Name=dyn_lightsaber1, swing1LimitDegrees=45, swing2LimitDegrees=45, twistLimitDegrees=15}
 {bone1Name=dyn_belt_01, bone2Name=pack, swing1LimitDegrees=20, swing2LimitDegrees=20, twistLimitDegrees=5}
 {bone1Name=dyn_belt_2, bone2Name=dyn_belt_01, swing1LimitDegrees=25, swing2LimitDegrees=25, twistLimitDegrees=5}
]
