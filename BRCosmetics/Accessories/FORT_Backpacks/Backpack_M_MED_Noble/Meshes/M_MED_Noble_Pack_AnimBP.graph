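# Animation Blueprint for the M_MED_Noble backpack, driving its dangling parts and toggling a material element.
# RigidBody physics, with a second front-end physics asset, surrounds additive bone adjustments to the handle, evaporator, knife and pack.
# Each update derives handle roll from pawn vertical velocity, with a correction that cancels it at zero velocity.
# Weapon, customization and skydiving events set "Hide Element 01" using NobleMale-tagged pickaxe checks and skydiving state.
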AnimBlueprint /BRCosmetics/Accessories/FORT_Backpacks/Backpack_M_MED_Noble/Meshes/M_MED_Noble_Pack_AnimBP
target_skeleton=&/M_MED_Noble_Pack_Skeleton
graphs [
 AnimGraph {nodes [
   n1 = RigidBody(ComponentPose=Local To Component(LocalPose=Local Space Ref Pose())) {
    bClampLinearTranslationLimitToRefPose=true
    bOverrideWorldGravity=true
    overridePhysicsAsset=&/M_MED_Noble_Pack_Physics
    simulationSpace=BaseBoneSpace
    alphaBoolBlend {blendOption=Linear}
    baseBoneRef {boneName=root}
    componentAppliedLinearAccClamp=<3000 3000 3000>
    componentLinearAccScale=<0.4 0.3 0.3>
    componentLinearVelScale=<0.4 0.3 0.3>
    overrideWorldGravity {z=-980}
   }
   n2 = "Transform (Modify) Bone - Bone: dyn_handle"(ComponentPose=@n1, Alpha=1, Translation="0,0,0", Rotation="0,0,-60", Scale="1,1,1") {rotationMode=BMM_Additive, alphaBoolBlend {blendOption=Linear}, boneToModify {boneName=dyn_handle}, rotation {roll=-60}}
   n3 = "Transform (Modify) Bone - Bone: dyn_handle"!(
    ComponentPose=@n2
    bAlphaBoolEnabled:bool=True
    Translation="0,0,0"
    Rotation_Roll:float=Get Fix() -> (Fix:double)
    Rotation_Pitch:float=0
    Rotation_Yaw:float=0
    Scale="1,1,1"
   ) -> (Pose) {alphaInputType=Bool, rotationMode=BMM_Additive, alphaBoolBlend {blendOption=Linear}, boneToModify {boneName=dyn_handle}}
   n4 = "Transform (Modify) Bone - Bone: dyn_handle"(
    ComponentPose=@n3
    bAlphaBoolEnabled=True
    Translation="0,0,0"
    Rotation_Roll=Get Rot() -> (Rot:double)
    Rotation_Pitch=0
    Rotation_Yaw=0
    Scale="1,1,1"
   ) {alphaInputType=Bool, rotationMode=BMM_Additive, alphaBoolBlend {blendOption=Linear}, boneToModify {boneName=dyn_handle}}
   n5 = "Transform (Modify) Bone - Bone: dyn_evaporator"!(ComponentPose=@n4, bAlphaBoolEnabled:bool=True, Translation="0,-7,-1", Rotation="10,-15,0", Scale="1,1,1") -> (Pose) {
    alphaInputType=Bool
    rotationMode=BMM_Additive
    translationMode=BMM_Additive
    translationSpace=BCS_BoneSpace
    alphaBoolBlend {blendOption=Linear}
    boneToModify {boneName=dyn_evaporator}
    rotation {pitch=10, yaw=-15}
    translation {y=-7, z=-1}
   }
   n6 = "Transform (Modify) Bone - Bone: dyn_knife"!(ComponentPose=@n5, bAlphaBoolEnabled:bool=True, Translation="-1,-3,2.5", Rotation="21.800,1.066,15.304", Scale="1,1,1") -> (Pose) {
    alphaInputType=Bool
    rotationMode=BMM_Additive
    translationMode=BMM_Additive
    translationSpace=BCS_BoneSpace
    alphaBoolBlend {blendOption=Linear}
    boneToModify {boneName=dyn_knife}
    rotation=<rot 21.8 1.066 15.304>
    translation=<-1 -3 2.5>
   }
   n7 = RigidBody!(ComponentPose=@n6, bAlphaBoolEnabled:bool=Get bIsInFrontEnd(), ExternalForce) -> (Pose) {
    alphaInputType=Bool
    bClampLinearTranslationLimitToRefPose=true
    bOverrideWorldGravity=true
    overridePhysicsAsset=&/M_MED_Noble_Pack_Physics_Frontend
    simulationSpace=BaseBoneSpace
    alphaBoolBlend {blendOption=Linear}
    baseBoneRef {boneName=root}
    componentAppliedLinearAccClamp=<3000 3000 3000>
    componentLinearAccScale=<0.4 0.3 0.3>
    componentLinearVelScale=<0.4 0.3 0.3>
    overrideWorldGravity {z=-980}
   }
   n8 = "Transform (Modify) Bone - Bone: pack"(ComponentPose=@n7, Translation="0,1.5,0", Rotation="0,0,0") {translationMode=BMM_Additive, alphaBoolBlend {blendOption=Linear}, boneToModify {boneName=pack}, translation {y=1.5}}
   Output Pose(Result=Component To Local(ComponentPose=@n8)) {layerGroup=None}
  ]}
 EventGraph {nodes [
   Event BlueprintUpdateAnimation()
   n1 = float + float(B=50)
   n2 = float + float(B=-630)
   n3 = Get PawnVelocityZ()
   n4 = float + float(A=@n3, B=@n2)
   n5 = float + float(B=-0.1)
   > Set Rot(Rot:double="Clamp (Float)"(Value=float * float(A=@n4, B=@n5) {is=title K2Node_PromotableOperator}, Min=0, Max=@n1)) -> (Output_Get:double)
   n6 = "Clamp (Float)"(Value=float * float(A=@n5, B=@n2), Min=0, Max=@n1)
   > Set Fix(Fix:double=SelectFloat(A=0, B=float * float(A=@n6, B=-1), bPickA="Not Equal (Float)"(A=@n3))) -> (Output_Get:double)
   n7 = Event OnWeaponChanged()
   n8 = Get Dynamic Material Instance(execute=[@n7.then, @n16.then], SkeletalMeshComponent=GetPartSkeletalMeshComponent())
   > n9 = Sequence()
   n10 = PickaxeItemDefinitionHasAnyMetaTag(Tags=Make Array("[0]"=(TagName="Cosmetics.MetaTags.NobleMale")), bRequiresEquipped=true)
   > n11 = Branch(Condition=@n10)
   > n12 = Branch(Condition=Get bIsSkyDiving())
   SetScalarParameterValue(execute=[@n14.then, @n12.then, @n13.else, @n11.else], self=@n8, ParameterName=Hide Element 01, Value=0)
   SetScalarParameterValue(execute=[@n12.else, @n13.then], self=@n8, ParameterName=Hide Element 01, Value=1)
   n13 = Branch(execute=[@n15.then, @n9.then_1], Condition=@n10)
   n14 = Event OnBeginSkydiving()
   n15 = Event OnEndSkydiving()
   n16 = Event OnCharacterCustomizationLoaded()
   TryGetPawnOwner()
   n17 = float + float(A="Sin (Radians)"(A=float * float(A=2)), B=2)
   "Clamp (Float)"(Value=@n17, Min=0, Max=1)
   PrintString!(execute, InString="To String (Float)"(InDouble=@n17), bPrintToScreen, bPrintToLog, TextColor, Duration, Key) -> (then)
  ]}
]
variables [
 Fix {type=number}
 Jumping {type=number}
 Rot {type=number}
]
